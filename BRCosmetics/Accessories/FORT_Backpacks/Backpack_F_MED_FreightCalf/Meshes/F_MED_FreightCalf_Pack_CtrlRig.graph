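# Control Rig for the FreightCalf backpack cosmetic that intermittently twitches the tops of its two upper arms.
# A looping timer drives a curve that makes a brief pulse early in each cycle.
# A seeded random value, held for the same duration as the loop, lets that pulse into Engage about a quarter of the time, otherwise zeroing it.
# Engage weights a fixed rotation offset on the left and right arm-top bones.
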ControlRigBlueprint /BRCosmetics/Accessories/FORT_Backpacks/Backpack_F_MED_FreightCalf/Meshes/F_MED_FreightCalf_Pack_CtrlRig
graphs [
 RigVMFunctionLibrary {links [], nodes []}
 RigVMModel {nodes [
   RigUnit_BeginExecution = &RigUnit_BeginExecution()
   LessEqual = /Script/RigVM.RigVMFunction_MathDoubleLessEqual(
    B=0
    A=/Script/RigVM.RigVMFunction_RandomFloat(Duration=1.5, Maximum=1, Minimum=-3, Seed=217) {is=struct RigVMUnitNode, template="Random::Execute(in Seed,in Minimum,in Maximum,in Duration,out Result)"}.Result
   ) {is=struct RigVMUnitNode, template="LessEqual::Execute(in A,in B,out Result)"}
   RigVMFunction_ControlFlowBranch = &RigVMFunction_ControlFlowBranch(Condition=@LessEqual.Result, ExecuteContext=@RigUnit_BeginExecution.ExecutePin)
   n1 = RigVMVariableNode() {variable=Engage}
   &RigUnit_ModifyTransforms(
    ItemToModify=((Item=(Type=Bone,Name="dyn_upper_l_arm_top"),Transform=(Rotation=(X=0,Y=-0.1736,Z=0,W=0.9848),Translation=(X=0,Y=0,Z=0),Scale3D=(X=1,Y=1,Z=1))))
    WeightMinimum=0
    ExecutePin=@RigVMFunction_ControlFlowBranch.Completed
    Weight=@n1.Value
   )
   > &RigUnit_ModifyTransforms(
    ItemToModify=((Item=(Type=Bone,Name="dyn_upper_r_arm_top"),Transform=(Rotation=(X=0,Y=-0.1736,Z=0,W=0.9848),Translation=(X=0,Y=0,Z=0),Scale3D=(X=1,Y=1,Z=1))))
    WeightMinimum=0
    Weight=@n1.Value
   )
   RigVMFunction_AnimEvalRichCurve = &RigVMFunction_AnimEvalRichCurve(
    Curve=(EditorCurveData=(Keys=((Time=0.05,Value=0.01),(Time=0.15,Value=1.01),(Time=0.25,Value=0.01))))
    SourceMinimum=0
    TargetMinimum=0
    Value=/Script/RigVM.RigVMFunction_TimeLoop(Duration=1.5, Normalize=true, Speed=1) {is=struct RigVMUnitNode}.Relative
   )
   RigVMVariableNode(Value=@RigVMFunction_AnimEvalRichCurve.Result, ExecuteContext=@RigVMFunction_ControlFlowBranch.False) {variable=Engage}
   RigVMVariableNode(ExecuteContext=@RigVMFunction_ControlFlowBranch.True) {variable=Engage}
  ]}
]
hierarchy [
 root {loc=<0 0 0>, rot=<0 0 0 1>}
 pack {parent=root, loc=<0 -0.2929 0>, rot=<0 0 0 1>}
 dyn_upper_l_arm_mstr {parent=pack, loc=<3.5878 -6.0066 -1.068>, rot=<0 0 0 1>}
 dyn_upper_l_arm {parent=dyn_upper_l_arm_mstr, loc=<0 0 0>, rot=<0 0 0 1>}
 dyn_upper_l_arm_top {parent=dyn_upper_l_arm, loc=<0 -4.1048 0>, rot=<0 0.5378 0 -0.8431>}
 dyn_upper_l_arm_bottom {parent=dyn_upper_l_arm, loc=<5.319 -0.4271 4.7124>, rot=<0 0.3563 0 -0.9344>}
 dyn_lower_l_arm_mstr {parent=pack, loc=<3.5878 -6.0066 -5.5932>, rot=<0 -0.342 0 -0.9397>}
 dyn_lower_l_arm_01 {parent=dyn_lower_l_arm_mstr, loc=<0 0 0>, rot=<0 0 0 1>}
 dyn_lower_l_arm_02 {parent=dyn_lower_l_arm_01, loc=<6.9078 -0.4271 0.0425>, rot=<0 0 0 1>}
 dyn_upper_r_arm_mstr {parent=pack, loc=<-3.5878 -6.0066 -1.068>, rot=<1 0 0 0>}
 dyn_upper_r_arm {parent=dyn_upper_r_arm_mstr, loc=<0 0 0>, rot=<0 -1 0 0>}
 dyn_upper_r_arm_top {parent=dyn_upper_r_arm, loc=<0 4.1048 0>, rot=<0 0.5378 0 -0.8431>}
 dyn_upper_r_arm_bottom {parent=dyn_upper_r_arm, loc=<5.319 0.4271 4.7124>, rot=<0 0.3563 0 -0.9344>}
 dyn_lower_r_arm_mstr {parent=pack, loc=<-3.5878 -6.0066 -5.5932>, rot=<-0.342 0 0.9397 0>}
 dyn_lower_r_arm_01 {parent=dyn_lower_r_arm_mstr, loc=<0 0 0>, rot=<0 0 0 1>}
 dyn_lower_r_arm_02 {parent=dyn_lower_r_arm_01, loc=<6.9078 0.4271 0.0425>, rot=<0 0 0 1>}
]
variables [
 AlphaRandomizer {type=double, default=0}
 Engage {type=double, default=0}
]
